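# Niagara skydiving contrail for the Starfield cosmetic, attached to the character's skeletal mesh.
# Active emitters include color-cycling body sprites, electric sprites and meshes at hands and feet, and a velocity-oriented dish mesh.
# FN_DivingAttributes supplies dive values that blend spawning, size and color, with separate front-end overrides.
# Several mesh trail, fire and speed cone emitters remain in the system but are disabled.
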NiagaraSystem /BRCosmetics/Athena/Effects/Particles/Skydiving/NS_Contrail_Starfield/NS_Contrail_Starfield
dependencies {
 usedDataInterfaces [&NiagaraDataInterfaceVectorCurve, &NiagaraDataInterfaceCurve, &NiagaraDataInterfaceColorCurve]
 usedDynamicInputs []
 usedModules [&SystemState, /Game/Effects/Niagara/Modules/Skydiving/FN_DivingAttributes/v4/FN_DivingAttributes_v4]
}
emitters {
 Blue_Main {
  bEnabled=false
  emitterUpdateScript [
   &EmitterState(Life Cycle Mode=::NewEnumerator1, Loop Duration=1, UseLoopDelay=false)
   &SpawnRate(SpawnRate=4.5, Use Spawn Probability=false)
   &SpawnRate(SpawnGroup=1, SpawnRate=4.5, Use Spawn Probability=false) {module=SpawnRate001}
  ]
  particleSpawnScript [
   /Game/Effects/Niagara/Modules/Skydiving/FN_DivingAttributes/FN_DivingAttributes_v2(Accumulate Dive=0.1, Accumulate Rest=-0.1, SkeletalMesh:NiagaraDataInterfaceSkeletalMesh=$User.SkeletalMesh)
   /SetVariables(Particles.DiveCLAMP:NiagaraFloat=&ClampFloat(Float=&FN_FrontEndFloatModifier(Gameplay=$Particles.Dive, Front End=&Multiply_Float(A=$Particles.Dive, B=10)))) {isSetParameters=true, module=SetVariables_#1}
   /SetVariables(Particles.MaterialRandom:NiagaraFloat=&UniformRangedFloat(Minimum=0, Maximum=1)) {isSetParameters=true, module=SetVariables_#2}
   /Niagara/Modules/Spawn/Initialization/InitializeParticle(
    Color=&LerpLinearColors(
     StartColor=<rgba 28.3334 0 50 0.25>
     EndColor=&LerpLinearColors(
      StartColor=<rgba 0.5667 0 1 0.4>
      EndColor=<rgba 0 40.4868 50 0.4>
      LerpFactor=&UniformAOrBFloat(A=0.5, B=0.7)
     )
     LerpFactor=$Particles.DiveCLAMP
    )
    Lifetime=&Lerp_Float(
     A=&UniformRangedFloat(Minimum=0.5, Maximum=0.7)
     B=&UniformRangedFloat(Minimum=0.8, Maximum=1)
     Alpha=$Particles.DiveCLAMP
    )
    Material Random:NiagaraFloat=$Particles.MaterialRandom
    Mesh Scale=&Multiply_Vector(A=<0.4 0.125 0.125>, B=&Lerp_Vector(A=<1 1 1>, B=<2 1 1>, Alpha=$Particles.DiveCLAMP))
    Module.Write Color=true
    Module.Write Lifetime=true
    Module.Write Mass=true
    Module.Write Material Random=true
    Module.Write Position=false
    Module.Write Scale=true
    Module.Write SpriteRotation=true
    Module.Write SpriteSize=false
    Sprite Rotation=&UniformRangedFloat(Minimum=-0.2, Maximum=0.2)
    Sprite Size=<5 5>
   )
   &SkeletalMeshLocation(
    Apply To Particle Position=false
    Bone / Socket Index=&SpawnGroup()
    Bone Sampling Mode=::NewEnumerator4
    Inherit Velocity=false
    MaskSpawnGroup=false
    Sampled Position Offset=<0 5 0>
    Skeletal Mesh=$User.SkeletalMesh
   ) {module=SkeletalMeshLocation001}
   /Niagara/Modules/Spawn/Orientation/V2/InitialMeshOrientation(
    Orientation Vector=<0 0 -1>
    Rotation=&UniformRangedVector(Minimum=<0 0 0>, Maximum=<1 0 0>)
    Use Orientation Vector=true
    Use Rotation Vector=true
   )
  ]
  particleUpdateScript [
   /SetVariables(Particles.DiveCLAMP:NiagaraFloat=&ClampFloat(Float=&FN_FrontEndFloatModifier(Gameplay=$Particles.Dive, Front End=&Multiply_Float(A=$Particles.Dive, B=10)))) {isSetParameters=true, module=SetVariables_#3}
   &ParticleState()
   &SkeletalMeshLocation(
    Bone / Socket Index=&SpawnGroup()
    Bone Sampling Mode=::NewEnumerator4
    Inherit Velocity=false
    MaskSpawnGroup=false
    Skeletal Mesh=$User.SkeletalMesh
   )
   /Niagara/Modules/Update/Orientation/V2/UpdateMeshOrientation(
    Delta Time:NiagaraFloat=&Multiply_Float(A=$Engine.DeltaTime, B=&Multiply_Float(A=5, B=$Particles.NormalizedAge))
    Rotation Rate:NiagaraFloat=$Particles.Initial.SpriteRotation
    Rotation Vector:Vector3f=<0 0 1>
   ) {enabled=false}
   &ScaleMeshSize(Scale Factor=&Multiply_Vector(
     A=&Lerp_Vector(
      A=&MakeVector(
       X=&FloatFromCurve(FloatCurve=DataInterface() {propertyValues {Curve {keys [<at 0 0.1>,<at 1 2>]}}})
       Y=&FloatFromCurve(FloatCurve=DataInterface() {propertyValues {Curve {keys [
            <at 0 1 RCIM_Cubic>
            {arriveTangent=-1.1222, interpMode=RCIM_Cubic, leaveTangent=-2.1367, tangentMode=RCTM_User, time=0.9, value=0.5}
            <at 1 0.05>
           ]}}})
       Z=&FloatFromCurve(FloatCurve=DataInterface() {propertyValues {Curve {keys [
            <at 0 1 RCIM_Cubic>
            {arriveTangent=-1.1222, interpMode=RCIM_Cubic, leaveTangent=-2.1367, tangentMode=RCTM_User, time=0.9, value=0.5}
            <at 1 0.05>
           ]}}})
      )
      B=&MakeVector(
       X=&FloatFromCurve(FloatCurve=DataInterface() {propertyValues {Curve {keys [<at 0 1>,<at 1 2>]}}})
       Y=&FloatFromCurve(FloatCurve=DataInterface() {propertyValues {Curve {keys [
            {arriveTangent=-1.5003, interpMode=RCIM_Cubic, leaveTangent=-0.2737, tangentMode=RCTM_User, time=0, value=1}
            {arriveTangent=-2.2772, interpMode=RCIM_Cubic, leaveTangent=-2.2772, tangentMode=RCTM_User, time=1, value=0}
           ]}}})
       Z=&FloatFromCurve(FloatCurve=DataInterface() {propertyValues {Curve {keys [
            {arriveTangent=-1.5003, interpMode=RCIM_Cubic, leaveTangent=-0.2737, tangentMode=RCTM_User, time=0, value=1}
            {arriveTangent=-2.2772, interpMode=RCIM_Cubic, leaveTangent=-2.2772, tangentMode=RCTM_User, time=1, value=0}
           ]}}})
      )
      Alpha=$Particles.DiveCLAMP
     )
     B=&VectorFromFloat(Value=&FloatFromCurve(FloatCurve=DataInterface() {propertyValues {Curve {keys [
           <at 0 0 RCIM_Constant>
           {arriveTangent=5.5862, interpMode=RCIM_Cubic, leaveTangent=5.5862, tangentMode=RCTM_User, time=0.3, value=0}
           <at 0.7 1>
           <at 1 1>
          ]}}}))
    )) {module=ScaleMeshSize001}
   &ScaleColor(
    Scale Alpha=&FloatFromCurve(FloatCurve=DataInterface() {propertyValues {Curve {keys [
         {arriveTangent=8.7245, interpMode=RCIM_Cubic, leaveTangent=8.7245, tangentMode=RCTM_User, time=0, value=0}
         <at 0.45 1 RCIM_Cubic>
         {arriveTangent=-0.5281, interpMode=RCIM_Cubic, leaveTangent=-0.5281, tangentMode=RCTM_User, time=1, value=0}
        ]}}})
    ScaleA=true
    ScaleRGB=true
   )
   &DynamicMaterialParameters(
    Index 0 Param 1=&Lerp_Float(A=0, B=0.5, Alpha=$Particles.DiveCLAMP)
    Module.Param0WriteEnabled=true
    Module.Param1WriteEnabled=false
    Module.Param2WriteEnabled=false
    Module.Param3WriteEnabled=false
   )
  ]
  rendererClasses [&NiagaraMeshRendererProperties]
  renderers [{rendererClass=&NiagaraMeshRendererProperties, rendererIndex=0}]
 }
 Blue_Main_Body {
  bEnabled=false
  emitterUpdateScript [
   &EmitterState(Life Cycle Mode=::NewEnumerator1, Loop Duration=1, UseLoopDelay=false)
   &SpawnRate(SpawnRate=5, Use Spawn Probability=false)
   &SpawnRate(SpawnGroup=1, SpawnRate=5, Use Spawn Probability=false) {module=SpawnRate001}
  ]
  particleSpawnScript [
   &FN_DivingAttributes_v2(Accumulate Dive=0.1, Accumulate Rest=-0.1, SkeletalMesh:NiagaraDataInterfaceSkeletalMesh=$User.SkeletalMesh)
   /SetVariables(Particles.DiveCLAMP:NiagaraFloat=&ClampFloat(Float=&FN_FrontEndFloatModifier(Gameplay=$Particles.Dive, Front End=&Multiply_Float(A=$Particles.Dive, B=10)))) {isSetParameters=true, module=SetVariables_#1}
   /SetVariables(Particles.MaterialRandom:NiagaraFloat=&UniformRangedFloat(Minimum=0, Maximum=1)) {isSetParameters=true, module=SetVariables_#2}
   &InitializeParticle(
    Color=&LerpLinearColors(
     StartColor=<rgba 28.3334 0 50 0.025>
     EndColor=&LerpLinearColors(
      StartColor=<rgba 0.3333 0 1 0.025>
      EndColor=<rgba 0 40.4868 50 0.03>
      LerpFactor=&UniformAOrBFloat(A=0.5, B=0.7)
     )
     LerpFactor=$Particles.DiveCLAMP
    )
    Lifetime=&Lerp_Float(
     A=&UniformRangedFloat(Minimum=0.5, Maximum=0.7)
     B=&UniformRangedFloat(Minimum=0.8, Maximum=1)
     Alpha=$Particles.DiveCLAMP
    )
    Material Random:NiagaraFloat=$Particles.MaterialRandom
    Mesh Scale=&Multiply_Vector(A=<0.755 0.32 0.32>, B=&Lerp_Vector(A=<1 1 1>, B=<2 1 1>, Alpha=$Particles.DiveCLAMP))
    Module.Write Color=true
    Module.Write Lifetime=true
    Module.Write Mass=true
    Module.Write Material Random=true
    Module.Write Position=false
    Module.Write Scale=true
    Module.Write SpriteRotation=true
    Module.Write SpriteSize=false
    Sprite Rotation=&UniformRangedFloat(Minimum=-0.2, Maximum=0.2)
    Sprite Size=<5 5>
   )
   &SkeletalMeshLocation(
    Apply To Particle Position=false
    Bone / Socket Index=6
    Bone Sampling Mode=::NewEnumerator4
    Inherit Velocity=false
    MaskSpawnGroup=false
    Sampled Position Offset=<0 5 0>
    Skeletal Mesh=$User.SkeletalMesh
   ) {enabled=false, module=SkeletalMeshLocation001}
   &InitialMeshOrientation(
    Orientation Vector=<0 0 -1>
    Rotation=&UniformRangedVector(Minimum=<0 0 0>, Maximum=<1 0 0>)
    Use Orientation Vector=true
    Use Rotation Vector=true
   )
  ]
  particleUpdateScript [
   /SetVariables(Particles.DiveCLAMP:NiagaraFloat=&ClampFloat(Float=&FN_FrontEndFloatModifier(Gameplay=$Particles.Dive, Front End=&Multiply_Float(A=$Particles.Dive, B=10)))) {isSetParameters=true, module=SetVariables_#3}
   &ParticleState()
   &SkeletalMeshLocation(
    Bone / Socket Index=6
    Bone Sampling Mode=::NewEnumerator4
    Inherit Velocity=false
    MaskSpawnGroup=false
    Sampled Position Offset=<0 8 0>
    Sampled Position Offset Coordinate Space:ENiagaraMeshTransforms=::NewEnumerator2
    Skeletal Mesh=$User.SkeletalMesh
   )
   &UpdateMeshOrientation(
    Delta Time:NiagaraFloat=&Multiply_Float(A=$Engine.DeltaTime, B=&Multiply_Float(A=5, B=$Particles.NormalizedAge))
    Rotation Rate:NiagaraFloat=$Particles.Initial.SpriteRotation
    Rotation Vector:Vector3f=<0 0 1>
   ) {enabled=false}
   &ScaleMeshSize(Scale Factor=&Multiply_Vector(
     A=&Lerp_Vector(
      A=&MakeVector(
       X=&FloatFromCurve(FloatCurve=DataInterface() {propertyValues {Curve {keys [<at 0 0.1>,<at 1 2>]}}})
       Y=&FloatFromCurve(FloatCurve=DataInterface() {propertyValues {Curve {keys [
            <at 0 1 RCIM_Cubic>
            {arriveTangent=-1.1222, interpMode=RCIM_Cubic, leaveTangent=-2.1367, tangentMode=RCTM_User, time=0.9, value=0.5}
            <at 1 0.05>
           ]}}})
       Z=&FloatFromCurve(FloatCurve=DataInterface() {propertyValues {Curve {keys [
            <at 0 1 RCIM_Cubic>
            {arriveTangent=-1.1222, interpMode=RCIM_Cubic, leaveTangent=-2.1367, tangentMode=RCTM_User, time=0.9, value=0.5}
            <at 1 0.05>
           ]}}})
      )
      B=&MakeVector(
       X=&FloatFromCurve(FloatCurve=DataInterface() {propertyValues {Curve {keys [<at 0 1>,<at 1 2>]}}})
       Y=&FloatFromCurve(FloatCurve=DataInterface() {propertyValues {Curve {keys [
            {arriveTangent=-1.5003, interpMode=RCIM_Cubic, leaveTangent=-0.2737, tangentMode=RCTM_User, time=0, value=1}
            {arriveTangent=-2.2772, interpMode=RCIM_Cubic, leaveTangent=-2.2772, tangentMode=RCTM_User, time=1, value=0}
           ]}}})
       Z=&FloatFromCurve(FloatCurve=DataInterface() {propertyValues {Curve {keys [
            {arriveTangent=-1.5003, interpMode=RCIM_Cubic, leaveTangent=-0.2737, tangentMode=RCTM_User, time=0, value=1}
            {arriveTangent=-2.2772, interpMode=RCIM_Cubic, leaveTangent=-2.2772, tangentMode=RCTM_User, time=1, value=0}
           ]}}})
      )
      Alpha=$Particles.DiveCLAMP
     )
     B=&VectorFromFloat(Value=&FloatFromCurve(FloatCurve=DataInterface() {propertyValues {Curve {keys [
           <at 0 0 RCIM_Constant>
           {arriveTangent=5.5862, interpMode=RCIM_Cubic, leaveTangent=5.5862, tangentMode=RCTM_User, time=0.3, value=0}
           <at 0.7 1>
           <at 1 1>
          ]}}}))
    )) {module=ScaleMeshSize001}
   &ScaleColor(
    Scale Alpha=&FloatFromCurve(FloatCurve=DataInterface() {propertyValues {Curve {keys [
         {arriveTangent=8.7245, interpMode=RCIM_Cubic, leaveTangent=8.7245, tangentMode=RCTM_User, time=0, value=0}
         <at 0.45 1 RCIM_Cubic>
         {arriveTangent=-0.5281, interpMode=RCIM_Cubic, leaveTangent=-0.5281, tangentMode=RCTM_User, time=1, value=0}
        ]}}})
    ScaleA=true
    ScaleRGB=true
   )
   &DynamicMaterialParameters(
    Index 0 Param 1=&Lerp_Float(A=0, B=0.5, Alpha=$Particles.DiveCLAMP)
    Module.Param0WriteEnabled=true
    Module.Param1WriteEnabled=false
    Module.Param2WriteEnabled=false
    Module.Param3WriteEnabled=false
   )
  ]
  rendererClasses [&NiagaraMeshRendererProperties]
  renderers [{rendererClass=&NiagaraMeshRendererProperties, rendererIndex=0}]
 }
 Body_Decal {
  emitterUpdateScript [
   &EmitterState(Life Cycle Mode=::NewEnumerator1, Loop Duration=1, UseLoopDelay=false)
   &SpawnRate(SpawnRate=20, Use Spawn Probability=false)
  ]
  particleSpawnScript [
   &FN_DivingAttributes_v2(Accumulate Dive=0.1, Accumulate Rest=-0.1, SkeletalMesh:NiagaraDataInterfaceSkeletalMesh=$User.SkeletalMesh)
   /SetVariables(Particles.DiveCLAMP:NiagaraFloat=&ClampFloat(Float=&FN_FrontEndFloatModifier(Gameplay=$Particles.Dive, Front End=&Multiply_Float(A=$Particles.Dive, B=10)))) {isSetParameters=true}
   /SetVariables(Particles.BodyNRG_dive=2, Particles.BodyNRG_rest=2) {isSetParameters=true}
   &InitializeParticle(
    Color=&MakeLinearColorFromVectorAndFloat(
     "Vector (RGB)"=&Lerp_Vector(
      A=&VectorFromCurve(
       VectorCurve=DataInterface() {propertyValues {
         XCurve {keys [<at 0.0094 0.0726>,<at 0.1453 0.2606>,<at 0.3657 1.0315>,<at 0.5814 0.9785>,<at 0.7855 0.141>]}
         YCurve {keys [<at 0.0069 0.1621>,<at 0.1313 0.9996>,<at 0.3723 0>,<at 0.5804 0.887>,<at 0.777 0.8672>,<at 0.9926 0.0706>]}
         ZCurve {keys [<at 0 1>,<at 0.1443 0.0988>,<at 0.7825 0.0307>,<at 0.9986 0.9855>]}
        }}
       CurveIndex=$Emitter.NormalizedLoopAge
       Scale Curve=<10 10 10>
      )
      B=&MakeVectorFromLinearColorRGB(Color=<rgba 0 3.6798 10 1>)
      Alpha=$Particles.DiveCLAMP
     )
     "Float (Alpha)"=&UniformRangedFloat(Minimum=0.6, Maximum=1)
    )
    Lifetime=&UniformRangedFloat(Minimum=0.6, Maximum=0.9)
    Module.Write Color=true
    Module.Write Lifetime=true
    Module.Write Mass=false
    Module.Write Material Random=true
    Module.Write Position=false
    Module.Write Scale=false
    Module.Write SpriteRotation=true
    Module.Write SpriteSize=true
    Position=&Add_Vector(A=/Niagara/DynamicInputs/Helpers/SimulationPosition(), B=<0 0 0>)
    Sprite Rotation=&UniformRangedFloat(Minimum=0, Maximum=180)
    Sprite Size=&Vector2DFromFloat(Value=&Lerp_Float(
      A=&UniformRangedFloat(Minimum=30, Maximum=50)
      B=&UniformRangedFloat(Minimum=65, Maximum=100)
      Alpha=$Particles.DiveCLAMP
     ))
   )
   /Niagara/Modules/Spawn/Location/SphereLocation(MaskSpawnGroup=false, Module.Override Randomness=false, Sphere Radius=30, Surface Only=false)
  ]
  particleUpdateScript [
   /SetVariables(Particles.DiveCLAMP:NiagaraFloat=&ClampFloat(Float=&FN_FrontEndFloatModifier(Gameplay=$Particles.Dive, Front End=&Multiply_Float(A=$Particles.Dive, B=10)))) {isSetParameters=true}
   &ParticleState()
   &ScaleColor(
    Scale Alpha=&FloatFromCurve(FloatCurve=DataInterface() {propertyValues {Curve {keys [<at 0 0 RCIM_Cubic>,<at 0.5 1 RCIM_Cubic>,<at 1 0>]}}})
    ScaleA=true
    ScaleRGB=true
   )
   &SkeletalMeshLocation(
    Bone Sampling Mode=::NewEnumerator0
    Inherit Velocity=false
    MaskSpawnGroup=false
    Module.Override Randomness=false
    Sampled Position Offset=<0 0 20>
    Skeletal Mesh=$User.SkeletalMesh
   )
   &DynamicMaterialParameters(
    Index 0 Param 1=&Lerp_Float(A=$Particles.BodyNRG_rest, B=$Particles.BodyNRG_dive, Alpha=$Particles.DiveCLAMP)
    Module.Param0WriteEnabled=true
    Module.Param1WriteEnabled=false
    Module.Param2WriteEnabled=false
    Module.Param3WriteEnabled=false
   ) {module=DynamicMaterialParameters001}
   &SpriteRotationRate(Rotation Rate=180)
   &ScaleSpriteSize(Scale Factor=&Vector2DFromFloat(Value=&FloatFromCurve(FloatCurve=DataInterface() {propertyValues {Curve {keys [<at 0 1 RCIM_Cubic>, {arriveTangent=-1.2435, interpMode=RCIM_Cubic, leaveTangent=-1.2435, tangentMode=RCTM_User, time=1, value=0.3}]}}})))
  ]
  rendererClasses [&NiagaraSpriteRendererProperties]
  renderers [{rendererClass=&NiagaraSpriteRendererProperties}]
 }
 Electric_Hands_and_Feet {
  emitterUpdateScript [
   &EmitterState()
   &SpawnRate(
    Spawn Probability=&Lerp_Float(
     A=&FloatFromCurve(
      FloatCurve=DataInterface() {propertyValues {Curve {keys [<at 0.0003 0 RCIM_Cubic>,<at 0.0924 0.8907 RCIM_Cubic>,<at 0.1958 0>,<at 1 0 RCIM_Cubic>]}}}
      CurveIndex=&Multiply_Float(A=$Emitter.NormalizedLoopAge, B=&RandomRangeFloat(Minimum=0.2, Maximum=0.5))
     )
     B=20
     Alpha=$System.Dive
    )
    SpawnRate=50
    Use Spawn Probability=true
   )
  ]
  particleSpawnScript [
   /Game/Effects/Niagara/Modules/Groups/FN_SetSpawnGroup_v1/FN_SetSpawnGroup(Emitter Bone Count=4)
   &InitializeParticle(
    Color=&MultiplyLinearColor(Linear Color=$User.ColorChange, Scale Factor=<2 2 2 1>)
    Lifetime=0.2
    Module.Write Color=true
    Module.Write Lifetime=true
    Module.Write Mass=true
    Module.Write Material Random=true
    Module.Write Position=false
    Module.Write Scale=false
    Module.Write SpriteRotation=true
    Module.Write SpriteSize=true
    Position=/Niagara/Functions/Localspace/SimulationPosition()
    Sprite Rotation=&RandomRangeFloat(Minimum=0, Maximum=1)
    Sprite Size=&Vector2DFromFloat(Value=&RandomRangeFloat(Minimum=5, Maximum=18))
   )
   /Niagara/Modules/Update/SubUV/SubUVAnimation(SubUV Animation Mode=::NewEnumerator1)
   /Game/Effects/Niagara/Modules/Math/FN_EvenNumberToSign(Input:NiagaraInt32=$Particles.SpawnGroup, Invert Sign=true, Output Float Attribute=true)
   /NiagaraFN/Sprite/FN_RandomizeSpriteUV()
  ]
  particleUpdateScript [
   &ParticleState()
   &SkeletalMeshLocation(
    Apply Sampled Orientation to Mesh Particles=true
    Bone / Socket Index=&Add_Integer(A=$Particles.SpawnGroup, B=9)
    Bone Sampling Mode=::NewEnumerator4
    Inherit Velocity=false
    Interpolate Mesh Sampling=false
    MaskSpawnGroup=false
    Sampled Position Offset=&MakeVector(
     X=&Multiply_Float(
      A=/Game/Effects/Niagara/DynamicInputs/Skydiving/SkydiveHandsFeetFloatModifier(
       Hands:NiagaraFloat=&FloatFromCurve(
        FloatCurve=DataInterface() {propertyValues {Curve {keys [
            {arriveTangent=1.5102, interpMode=RCIM_Cubic, leaveTangent=1.51, tangentMode=RCTM_User, time=0, value=10}
            {arriveTangent=40.7212, interpMode=RCIM_Cubic, leaveTangent=40.7212, tangentMode=RCTM_User, time=1, value=30}
           ]}}}
        CurveIndex=&RandomRangeFloat(Minimum=0, Maximum=1)
       )
       Feet:NiagaraFloat=&RandomRangeFloat(Minimum=-20, Maximum=0)
      )
      B=$Particles.Sign
     )
     Y=0
     Z=0
    )
    Sampled Position Offset Coordinate Space:ENiagaraMeshTransforms=::NewEnumerator3
    Skeletal Mesh=$User.SkeletalMesh
   )
   &ScaleSpriteSize(Scale Factor=&Vector2DFromFloat(Value=&FloatFromCurve(FloatCurve=DataInterface() {propertyValues {Curve {keys [<at 0 1>,<at 1 2>]}}})))
  ]
  rendererClasses [&NiagaraSpriteRendererProperties, &NiagaraMeshRendererProperties]
  renderers [
   {rendererClass=&NiagaraSpriteRendererProperties}
   {rendererClass=&NiagaraMeshRendererProperties, rendererIndex=1}
  ]
 }
 Fire {
  bEnabled=false
  emitterUpdateScript [
   &EmitterState(Life Cycle Mode=::NewEnumerator1, Loop Duration=1, UseLoopDelay=false)
   &SpawnRate(SpawnRate=2, Use Spawn Probability=false)
   &SpawnRate(SpawnGroup=1, SpawnRate=2, Use Spawn Probability=false) {module=SpawnRate001}
  ]
  particleSpawnScript [
   &FN_DivingAttributes_v2(Accumulate Dive=0.1, Accumulate Rest=-0.1, SkeletalMesh:NiagaraDataInterfaceSkeletalMesh=$User.SkeletalMesh)
   /SetVariables(Particles.DiveCLAMP:NiagaraFloat=&ClampFloat(Float=&FN_FrontEndFloatModifier(Gameplay=$Particles.Dive, Front End=&Multiply_Float(A=$Particles.Dive, B=10)))) {isSetParameters=true, module=SetVariables_#1}
   &InitializeParticle(
    Color=&LerpLinearColors(StartColor=<rgba 28.3334 0 50 0.2>, EndColor=<rgba 0 40.4868 50 0.2>, LerpFactor=$Particles.DiveCLAMP)
    Lifetime=&Lerp_Float(
     A=&UniformRangedFloat(Minimum=0.8, Maximum=1)
     B=&UniformRangedFloat(Minimum=0.5, Maximum=0.7)
     Alpha=$Particles.DiveCLAMP
    )
    Module.Write Color=true
    Module.Write Lifetime=true
    Module.Write Mass=false
    Module.Write Material Random=true
    Module.Write Position=false
    Module.Write Scale=false
    Module.Write SpriteRotation=true
    Module.Write SpriteSize=true
    Position=&SimulationPosition()
    Sprite Rotation=&UniformRangedFloat(Minimum=-180, Maximum=180)
    Sprite Size=&Vector2DFromFloat(Value=&Lerp_Float(
      A=&UniformRangedFloat(Minimum=35, Maximum=32)
      B=&UniformRangedFloat(Minimum=35, Maximum=40)
      Alpha=$Particles.DiveCLAMP
     ))
   )
   &DynamicMaterialParameters(
    Index 0 Param 2=&UniformRangedFloat(Minimum=0, Maximum=1)
    Index 0 Param 3=1
    Index 0 Param 4=&UniformRangedFloat(Minimum=1, Maximum=2)
    Module.Param0WriteEnabled=false
    Module.Param1WriteEnabled=true
    Module.Param2WriteEnabled=false
    Module.Param3WriteEnabled=true
   ) {module=DynamicMaterialParameters002}
   &AddVelocity(Velocity=&Lerp_Vector(
     A=&UniformRangedVector(Minimum=<-25 0 0>, Maximum=<25 0 0>)
     B=&UniformRangedVector(Minimum=<-50 0 0>, Maximum=<50 0 0>)
     Alpha=$Particles.DiveCLAMP
    )) {enabled=false}
   &CameraOffset(Camera Offset Amount=35)
  ]
  particleUpdateScript [
   /SetVariables(Particles.DiveCLAMP:NiagaraFloat=&ClampFloat(Float=&FN_FrontEndFloatModifier(Gameplay=$Particles.Dive, Front End=&Multiply_Float(A=$Particles.Dive, B=10)))) {isSetParameters=true, module=SetVariables_#3}
   &ParticleState()
   &SphereLocation(
    Hemisphere X=true
    MaskSpawnGroup=false
    Module.Override Randomness=false
    Sphere Origin:NiagaraPosition=/Niagara/DynamicInputs/Helpers/SimulationPosition()
    Sphere Radius=10
    Surface Only=false
   ) {enabled=false}
   &SkeletalMeshLocation(
    Bone / Socket Index=&SpawnGroup()
    Bone Sampling Mode=::NewEnumerator4
    Inherit Velocity=false
    "Inherit Velocity (Scale)"=0
    MaskSpawnGroup=false
    Skeletal Mesh=$User.SkeletalMesh
   )
   &Drag(Use Linear Drag=true, Use Rotational Drag=false)
   &SolveForcesAndVelocity(Clamp Velocity=false, Limit Acceleration=false)
   &SubUVAnimation(Number Of Frames=64)
   &ScaleSpriteSize(Scale Factor=&Vector2DFromFloat(Value=&FloatFromCurve(FloatCurve=DataInterface() {propertyValues {Curve {keys [
          {arriveTangent=7.5764, interpMode=RCIM_Cubic, time=0, value=0.02}
          {arriveTangent=1.2063, interpMode=RCIM_Cubic, leaveTangent=1.2063, tangentMode=RCTM_User, time=0.3, value=0.99}
          <at 1 0.5 RCIM_Cubic>
         ]}}})))
   &DynamicMaterialParameters(
    Index 0 Param 1=&FloatFromCurve(FloatCurve=DataInterface() {propertyValues {Curve {keys [{arriveTangent=2.7551, interpMode=RCIM_Cubic, leaveTangent=2.7551, tangentMode=RCTM_User, time=0, value=0}, <at 0.7 1>]}}})
    Index 0 Param 2=1
    Index 0 Param 3=1
    Index 0 Param 4=1
    Module.Param0WriteEnabled=true
    Module.Param1WriteEnabled=true
    Module.Param2WriteEnabled=true
    Module.Param3WriteEnabled=false
   ) {module=DynamicMaterialParameters001}
   &ScaleColor(
    Scale Alpha=&FloatFromCurve(FloatCurve=DataInterface() {propertyValues {Curve {keys [
         {arriveTangent=8.7245, interpMode=RCIM_Cubic, leaveTangent=8.7245, tangentMode=RCTM_User, time=0, value=0}
         <at 0.2 1 RCIM_Cubic>
         {arriveTangent=-0.5281, interpMode=RCIM_Cubic, leaveTangent=-0.5281, tangentMode=RCTM_User, time=1, value=0}
        ]}}})
    ScaleA=true
    ScaleRGB=true
   )
  ]
  rendererClasses [&NiagaraSpriteRendererProperties]
  renderers [{rendererClass=&NiagaraSpriteRendererProperties}]
 }
 Fire_Upward {
  bEnabled=false
  emitterUpdateScript [
   &EmitterState(Life Cycle Mode=::NewEnumerator1, Loop Duration=1, UseLoopDelay=false)
   &SpawnRate(SpawnRate=6, Use Spawn Probability=false)
   &SpawnRate(SpawnGroup=1, SpawnRate=6, Use Spawn Probability=false) {module=SpawnRate001}
   &SpawnPerUnit(Module.Use Max Movement Threshold=true, Module.Use Movement Tolerance=false, Spawn Per Unit=400, Use Spawn Probability=false)
   &SpawnPerUnit(Module.Use Max Movement Threshold=true, Module.Use Movement Tolerance=false, Spawn Group=1, Spawn Per Unit=400, Use Spawn Probability=false) {module=SpawnPerUnit001}
  ]
  particleSpawnScript [
   &FN_DivingAttributes_v2(SkeletalMesh:NiagaraDataInterfaceSkeletalMesh=$User.SkeletalMesh)
   /SetVariables(Particles.DiveCLAMP:NiagaraFloat=&ClampFloat(Float=&FN_FrontEndFloatModifier(Gameplay=$Particles.Dive, Front End=&Multiply_Float(A=$Particles.Dive, B=10)))) {isSetParameters=true, module=SetVariables_#1}
   &InitializeParticle(
    Color=&LerpLinearColors(StartColor=<rgba 28.3334 0 50 0.15>, EndColor=<rgba 0 40.4868 50 0.025>, LerpFactor=$Particles.DiveCLAMP)
    Lifetime=&Lerp_Float(
     A=&UniformRangedFloat(Minimum=0.5, Maximum=0.3)
     B=&UniformRangedFloat(Minimum=0.35, Maximum=0.25)
     Alpha=$Particles.DiveCLAMP
    )
    Module.Write Color=true
    Module.Write Lifetime=true
    Module.Write Mass=false
    Module.Write Material Random=true
    Module.Write Position=false
    Module.Write Scale=false
    Module.Write SpriteRotation=true
    Module.Write SpriteSize=true
    Position=/Niagara/Functions/Localspace/SimulationPosition()
    Sprite Rotation=&UniformRangedFloat(Minimum=-360, Maximum=360)
    Sprite Size=&Vector2DFromFloat(Value=&Lerp_Float(
      A=&UniformRangedFloat(Minimum=25, Maximum=35)
      B=&UniformRangedFloat(Minimum=25, Maximum=22)
      Alpha=$Particles.DiveCLAMP
     ))
   )
   &DynamicMaterialParameters(
    Index 0 Param 2=&UniformRangedFloat(Minimum=0, Maximum=1)
    Index 0 Param 3=1
    Index 0 Param 4=&UniformRangedFloat(Minimum=1, Maximum=2)
    Module.Param0WriteEnabled=false
    Module.Param1WriteEnabled=true
    Module.Param2WriteEnabled=false
    Module.Param3WriteEnabled=true
   ) {module=DynamicMaterialParameters002}
   &SphereLocation(
    Hemisphere X=true
    MaskSpawnGroup=false
    Module.Override Randomness=false
    Sphere Origin:NiagaraPosition=/Niagara/DynamicInputs/Helpers/SimulationPosition()
    Sphere Radius=10
    Surface Only=false
   ) {enabled=false}
   &SkeletalMeshLocation(
    Bone / Socket Index=&SpawnGroup()
    Bone Sampling Mode=::NewEnumerator4
    Inherit Velocity=false
    "Inherit Velocity (Scale)"=0
    MaskSpawnGroup=false
    Sampled Position Offset Coordinate Space:ENiagaraMeshTransforms=::NewEnumerator1
    Skeletal Mesh=$User.SkeletalMesh
   )
   &AddVelocity(Velocity=&Lerp_Vector(
     A=&UniformRangedVector(Minimum=<-25 0 0>, Maximum=<25 0 0>)
     B=&UniformRangedVector(Minimum=<-50 0 0>, Maximum=<50 0 0>)
     Alpha=$Particles.DiveCLAMP
    )) {enabled=false}
   &AddVelocity(
    Coordinate Space=::World
    Scale Added Velocity=&VectorFromFloat(Value=&Lerp_Float(A=1, B=1, Alpha=$Particles.DiveCLAMP))
    Velocity=&UniformRangedVector(Minimum=<0 0 100>, Maximum=<0 0 70>)
   ) {module=AddVelocity001}
   &CameraOffset(Camera Offset Amount=35)
  ]
  particleUpdateScript [
   /SetVariables(Particles.DiveCLAMP:NiagaraFloat=&ClampFloat(Float=&FN_FrontEndFloatModifier(Gameplay=$Particles.Dive, Front End=&Multiply_Float(A=$Particles.Dive, B=10)))) {isSetParameters=true, module=SetVariables_#3}
   &ParticleState()
   &Drag(Drag=&Lerp_Float(A=0, B=6, Alpha=$Particles.DiveCLAMP), Use Linear Drag=true, Use Rotational Drag=false) {enabled=false}
   &AccelerationForce(
    Acceleration=&Multiply_VectorByFloat(Vector=<0 0 200>, Float=&UniformRangedFloat(Minimum=0.5, Maximum=1))
    Coordinate Space=::Local
   )
   &SolveForcesAndVelocity(Clamp Velocity=false, Limit Acceleration=false)
   &SubUVAnimation(Number Of Frames=64)
   &ScaleSpriteSize(Scale Factor=&Vector2DFromFloat(Value=&FloatFromCurve(FloatCurve=DataInterface() {propertyValues {Curve {keys [
          {arriveTangent=-0.5754, interpMode=RCIM_Cubic, leaveTangent=-0.5754, tangentMode=RCTM_User, time=0, value=0.99}
          {arriveTangent=-0.224, interpMode=RCIM_Cubic, leaveTangent=-0.224, tangentMode=RCTM_User, time=1, value=0.5}
         ]}}})))
   &DynamicMaterialParameters(
    Index 0 Param 1=&FloatFromCurve(FloatCurve=DataInterface() {propertyValues {Curve {keys [{arriveTangent=2.7551, interpMode=RCIM_Cubic, leaveTangent=2.7551, tangentMode=RCTM_User, time=0, value=0}, <at 0.7 1>]}}})
    Index 0 Param 2=1
    Index 0 Param 3=1
    Index 0 Param 4=1
    Module.Param0WriteEnabled=true
    Module.Param1WriteEnabled=true
    Module.Param2WriteEnabled=true
    Module.Param3WriteEnabled=false
   ) {module=DynamicMaterialParameters001}
   &ScaleColor(
    Scale Alpha=&FloatFromCurve(FloatCurve=DataInterface() {propertyValues {Curve {keys [
         {arriveTangent=8.7245, interpMode=RCIM_Cubic, leaveTangent=8.7245, tangentMode=RCTM_User, time=0, value=0}
         <at 0.2 1 RCIM_Cubic>
         {arriveTangent=-0.5281, interpMode=RCIM_Cubic, leaveTangent=-0.5281, tangentMode=RCTM_User, time=1, value=0}
        ]}}})
    ScaleA=true
    ScaleRGB=true
   )
  ]
  rendererClasses [&NiagaraSpriteRendererProperties]
  renderers [{rendererClass=&NiagaraSpriteRendererProperties}]
 }
 Mesh_Dish {
  emitterUpdateScript [
   &EmitterState(Life Cycle Mode=::NewEnumerator1, Loop Duration=1, UseLoopDelay=false)
   &SpawnRate(
    SpawnRate=&Lerp_Float(
     A=&FN_FrontEndFloatModifier(Gameplay=0, Front End=0)
     B=&FN_FrontEndFloatModifier(Gameplay=3, Front End=0)
     Alpha=$System.Dive
    )
    Use Spawn Probability=false
   )
  ]
  particleSpawnScript [
   &InitializeParticle(
    Mesh Scale=<1 0.5 0.5>
    Module.Write Color=true
    Module.Write Lifetime=true
    Module.Write Mass=true
    Module.Write Material Random=true
    Module.Write Position=false
    Module.Write Scale=true
    Module.Write SpriteRotation=false
    Module.Write SpriteSize=false
    Position=/Niagara/Functions/Localspace/SimulationPosition()
    Sprite Size=<5 5>
   )
   &InitialMeshOrientation(
    Orientation Vector=<0 0 1>
    Rotation=&MakeVector(X=&RandomRangeFloat(Minimum=0, Maximum=0.1), Y=0, Z=0)
    Use Orientation Vector=false
    Use Rotation Vector=true
   )
  ]
  particleUpdateScript [
   &ParticleState()
   &AddVelocity(
    Coordinate Space=::World
    Scale Added Velocity=&VectorFromFloat(Value=&Lerp_Float(A=-10, B=-50, Alpha=$System.Dive))
    Velocity=&NormalizeVector(Vector To Normalize=$Engine.Owner.Velocity)
   )
   &AddVelocity(Velocity=&FN_FrontEndVector3Modifier(Gameplay=<0 0 0>, Front End=<0 0 20>)) {module=AddVelocity001}
   &SkeletalMeshLocation(
    Bone / Socket Index=6
    Bone Sampling Mode=::NewEnumerator4
    Inherit Velocity=false
    MaskSpawnGroup=false
    Sampled Position Offset=&Multiply_Vector(
     A=&NormalizeVector(Vector To Normalize=$Particles.Velocity)
     B=&FN_FrontEndVector3Modifier(Gameplay=&Lerp_Vector(A=<-20 -20 -20>, B=<-90 -90 -90>, Alpha=$System.Dive), Front End=<0 0 -18>)
    )
    Sampled Position Offset Coordinate Space:ENiagaraMeshTransforms=::NewEnumerator2
    Skeletal Mesh=$User.SkeletalMesh
   )
   &SolveForcesAndVelocity(Clamp Velocity=false, Limit Acceleration=false)
   &ScaleMeshSize(Scale Factor=&ScaleFloatByCurve(
     InitialValue:Vector3f=<1 1 1>
     ScaleCurve:NiagaraDataInterfaceVectorCurve=DataInterface() {propertyValues {XCurve {keys [<at 0 0.25>,<at 1 3>]}, YCurve {keys [<at 0 1.25>,<at 1 3.5>]}, ZCurve {keys [<at 0 1.25>,<at 1 3.5>]}}}
     CurveIndex:NiagaraFloat=$Particles.NormalizedAge
    ))
   &ScaleMeshSize(Scale Factor=&Multiply_Vector(
     A=&Lerp_Vector(A=<0.1 0.1 0.1>, B=<1 0.3 0.3>, Alpha=$System.Dive)
     B=&FN_FrontEndVector3Modifier(Gameplay=<1 1 1>, Front End=<5 5 3>)
    )) {module=ScaleMeshSize001}
   &DynamicMaterialParameters(
    Index 0 Param 1=3
    Index 0 Param 2=1
    Module.Param0WriteEnabled=true
    Module.Param1WriteEnabled=true
    Module.Param2WriteEnabled=true
    Module.Param3WriteEnabled=true
   )
   &ScaleColor(
    Scale Alpha=&FloatFromCurve(
     FloatCurve=DataInterface() {propertyValues {Curve {keys [
         <at 0 0 RCIM_Cubic>
         <at 0.5 1 RCIM_Cubic>
         {arriveTangent=0.0656, interpMode=RCIM_Cubic, leaveTangent=0.0656, tangentMode=RCTM_User, time=1, value=0}
        ]}}}
     Scale Curve=0.1
    )
    Scale RGB=&MakeVectorFromLinearColorRGB(Color=&MakeLinearColorFromVectorAndFloat(
      "Vector (RGB)"=&Lerp_Vector(
       A=&VectorFromCurve(
        VectorCurve=DataInterface() {propertyValues {
          XCurve {keys [<at 0.0094 0.0726>,<at 0.1453 0.2606>,<at 0.3657 1.0315>,<at 0.5814 0.9785>,<at 0.7855 0.141>]}
          YCurve {keys [<at 0.0069 0.1621>,<at 0.1313 0.9996>,<at 0.3723 0>,<at 0.5804 0.887>,<at 0.777 0.8672>,<at 0.9926 0.0706>]}
          ZCurve {keys [<at 0 1>,<at 0.1443 0.0988>,<at 0.7825 0.0307>,<at 0.9986 0.9855>]}
         }}
        CurveIndex=$Emitter.NormalizedLoopAge
       )
       B=&VectorFromCurve(
        VectorCurve=DataInterface() {propertyValues {
          XCurve {keys [<at 0.0094 0.0726>,<at 0.1453 0.2606>,<at 0.3657 1.0315>,<at 0.5814 0.9785>,<at 0.7855 0.141>]}
          YCurve {keys [<at 0.0069 0.1621>,<at 0.1313 0.9996>,<at 0.3723 0>,<at 0.5804 0.887>,<at 0.777 0.8672>,<at 0.9926 0.0706>]}
          ZCurve {keys [<at 0 1>,<at 0.1443 0.0988>,<at 0.7825 0.0307>,<at 0.9986 0.9855>]}
         }}
        CurveIndex=$Emitter.NormalizedLoopAge
        Scale Curve=<5 5 5>
       )
       Alpha=$System.Dive
      )
      "Float (Alpha)"=&UniformRangedFloat(Minimum=0.6, Maximum=1)
     ))
    ScaleA=true
    ScaleRGB=true
   ) {enabled=false}
   /Niagara/Modules/Update/Orientation/V2/OrientMeshToVector(Look At Direction:Vector3f=$Particles.Velocity, Override Quat=false, UseOptionalSideAxis=false, UseOptionalUpAxis=false)
   &ScaleColor(
    Scale Alpha=&FloatFromCurve(FloatCurve=DataInterface() {propertyValues {Curve {keys [
         <at 0 0 RCIM_Cubic>
         <at 0.5 1 RCIM_Cubic>
         {arriveTangent=0.0656, interpMode=RCIM_Cubic, leaveTangent=0.0656, tangentMode=RCTM_User, time=1, value=0}
        ]}}})
    Scale RGB=&MakeVectorFromLinearColorRGB(Color=&MakeLinearColorFromVectorAndFloat(
      "Vector (RGB)"=&Lerp_Vector(
       A=<5 1 1>
       B=&VectorFromCurve(
        VectorCurve=DataInterface() {propertyValues {XCurve {keys [<at 0 1>,<at 1 0>]}, YCurve {keys [<at 0 1>,<at 1 0>]}, ZCurve {keys [<at 0 1>,<at 1 0>]}}}
        Scale Curve=<5 5 5>
       )
       Alpha=$System.Dive
      )
      "Float (Alpha)"=&UniformRangedFloat(Minimum=0.6, Maximum=1)
     ))
    ScaleA=true
    ScaleRGB=true
   ) {enabled=false, module=ScaleColor001}
   &ScaleColor(
    Scale Alpha=&FloatFromCurve(
     FloatCurve=DataInterface() {propertyValues {Curve {keys [
         <at 0 0 RCIM_Cubic>
         <at 0.5 1 RCIM_Cubic>
         {arriveTangent=0.0656, interpMode=RCIM_Cubic, leaveTangent=0.0656, tangentMode=RCTM_User, time=1, value=0}
        ]}}}
     Scale Curve=0.5
    )
    Scale RGB=&MakeVectorFromLinearColorRGB(Color=&MultiplyLinearColor(
      Linear Color=&ColorFromCurve(
       ColorCurve=DataInterface() {propertyValues {
         AlphaCurve {keys [<at 0 1 RCIM_Cubic>,<at 1 1 RCIM_Cubic>]}
         BlueCurve {keys [<at 0.0016 0.1372>,<at 0.5777 1>,<at 1 0.1372 RCIM_Cubic>]}
         GreenCurve {keys [<at 0.0016 0.05>,<at 0.5777 0.338>,<at 1 0.05 RCIM_Cubic>]}
         RedCurve {keys [<at 0.0016 1>,<at 0.5777 0>,<at 1 1 RCIM_Cubic>]}
        }}
       CurveIndex=$Emitter.LoopedAge
      )
      Scale Factor=<5 5 5 1>
     ))
    ScaleA=true
    ScaleRGB=true
   ) {module=ScaleColor002}
  ]
  rendererClasses [&NiagaraMeshRendererProperties]
  renderers [{rendererClass=&NiagaraMeshRendererProperties, rendererIndex=0}]
 }
 SpeedCone_Main {
  bEnabled=false
  emitterUpdateScript [
   &EmitterState(Life Cycle Mode=::NewEnumerator1, Loop Duration=1, UseLoopDelay=false)
   &SpawnRate(SpawnRate=&Lerp_Float(A=3, B=3, Alpha=$System.Dive), Use Spawn Probability=false)
  ]
  particleSpawnScript [
   &FN_DivingAttributes_v2(Accumulate Dive=0.1, Accumulate Rest=-0.1, SkeletalMesh:NiagaraDataInterfaceSkeletalMesh=$User.SkeletalMesh)
   /SetVariables(Particles.DiveAccumulationMax=40, Particles.DiveAccumulationMin=-5) {isSetParameters=true}
   /SetVariables(Particles.DiveCLAMP:NiagaraFloat=&ClampFloat(Float=&FN_FrontEndFloatModifier(Gameplay=$Particles.Dive, Front End=&Multiply_Float(A=$Particles.Dive, B=10)))) {isSetParameters=true}
   &InitializeParticle(
    Color=&LerpLinearColors(
     StartColor=<rgba 0.5667 0 1 1.5>
     EndColor=&LerpLinearColors(StartColor=<rgba 0.5667 0 1 1>, EndColor=<rgba 0 0.6167 1 1>, LerpFactor=&UniformAOrBFloat(A=0.5, B=0.7))
     LerpFactor=$Particles.DiveCLAMP
    )
    Mesh Scale=&Multiply_Vector(
     A=<0.5 0.6 0.6>
     B=&Lerp_Vector(
      A=&UniformRangedVector(Minimum=<0.3 0.7 0.5>, Maximum=<0.4 0.8 0.6>)
      B=&UniformRangedVector(Minimum=<1 1.5 1.2>, Maximum=<1 1.5 1.3>)
      Alpha=$Particles.Dive
     )
    )
    Module.Write Color=true
    Module.Write Lifetime=true
    Module.Write Mass=false
    Module.Write Material Random=true
    Module.Write Position=false
    Module.Write Scale=true
    Module.Write SpriteRotation=false
    Module.Write SpriteSize=false
    Sprite Size=<5 5>
   )
   &InitialMeshOrientation(
    Orientation Vector=<0 0 1>
    Rotation=&Lerp_Vector(A=<0 0 -0.03>, B=<0 0 0>, Alpha=$Particles.Dive)
    Use Orientation Vector=true
    Use Rotation Vector=true
   )
  ]
  particleUpdateScript [
   /SetVariables(Particles.DivingAttributesExt:NiagaraFloat=&Lerp_Float(A=-5, B=0.2, Alpha=&ClampFloat(Float=&Lerp_Float(A=-1, B=2, Alpha=$Particles.Dive)))) {isSetParameters=true}
   /SetVariables(Particles.Dive++:NiagaraFloat=&ClampFloat(
     Float=&Add_Float(A=$Particles.DiveAccumulation, B=$Particles.DivingAttributesExt)
     Min=$Particles.Initial.DiveAccumulationMin
     Max=$Particles.Initial.DiveAccumulationMax
    )) {isSetParameters=true}
   /SetVariables(Particles.DiveAccumulation:NiagaraFloat=Linked() {linkedVariable {name=Particles.Dive++, type=NiagaraFloat}}) {isSetParameters=true}
   /SetVariables(Particles.DiveCLAMP:NiagaraFloat=&ClampFloat(Float=&FN_FrontEndFloatModifier(Gameplay=$Particles.Dive, Front End=&Multiply_Float(A=$Particles.Dive, B=10)))) {isSetParameters=true}
   &ParticleState()
   &SkeletalMeshLocation(
    Bone / Socket Index=6
    Bone Sampling Mode=::NewEnumerator4
    Inherit Velocity=false
    MaskSpawnGroup=false
    Sampled Position Offset=&Lerp_Vector(A=<0 10 -16>, B=<0 10 -25>, Alpha=$Particles.Dive)
    Sampled Position Offset Coordinate Space:ENiagaraMeshTransforms=::NewEnumerator2
    Skeletal Mesh=$User.SkeletalMesh
   )
   &SolveForcesAndVelocity(Clamp Velocity=false, Limit Acceleration=false)
   &ScaleColor(
    Scale Alpha=&FloatFromCurve(FloatCurve=DataInterface() {propertyValues {Curve {keys [
         <at 0 0 RCIM_Cubic>
         <at 0.5 1 RCIM_Cubic>
         {arriveTangent=0.0656, interpMode=RCIM_Cubic, leaveTangent=0.0656, tangentMode=RCTM_User, time=1, value=0}
        ]}}})
    Scale RGB=&VectorFromFloat(Value=&Add_Float(
      A=1
      B=&Multiply_Float(A=$Particles.DiveAccumulation, B=&Divide_Float(A=1, B=$Particles.Initial.DiveAccumulationMax))
     ))
    ScaleA=true
    ScaleRGB=true
   ) {enabled=false}
   &ScaleMeshSize(Scale Factor=&VectorFromFloat(Value=0.5)) {module=ScaleMeshSize001}
   &ScaleMeshSize(Scale Factor=&ScaleFloatByCurve(
     InitialValue:Vector3f=<1 1 1>
     ScaleCurve:NiagaraDataInterfaceVectorCurve=DataInterface() {propertyValues {XCurve {keys [<at 0 0.25>,<at 1 3>]}, YCurve {keys [<at 0 1.25>,<at 1 3.5>]}, ZCurve {keys [<at 0 1.25>,<at 1 3.5>]}}}
     CurveIndex:NiagaraFloat=$Particles.NormalizedAge
    ))
   &ScaleColor(
    Scale RGB=&MakeVectorFromLinearColorRGB(Color=&MakeLinearColorFromVectorAndFloat(
      "Vector (RGB)"=&Lerp_Vector(
       A=&VectorFromCurve(
        VectorCurve=DataInterface() {propertyValues {
          XCurve {keys [<at 0.0094 0.0726>,<at 0.1453 0.2606>,<at 0.3657 1.0315>,<at 0.5814 0.9785>,<at 0.7855 0.141>]}
          YCurve {keys [<at 0.0069 0.1621>,<at 0.1313 0.9996>,<at 0.3723 0>,<at 0.5804 0.887>,<at 0.777 0.8672>,<at 0.9926 0.0706>]}
          ZCurve {keys [<at 0 1>,<at 0.1443 0.0988>,<at 0.7825 0.0307>,<at 0.9986 0.9855>]}
         }}
        CurveIndex=$Emitter.NormalizedLoopAge
        Scale Curve=<10 10 10>
       )
       B=&MakeVectorFromLinearColorRGB(Color=<rgba 0 3.6798 10 1>)
       Alpha=$Particles.DiveCLAMP
      )
      "Float (Alpha)"=&UniformRangedFloat(Minimum=0.6, Maximum=1)
     ))
    ScaleA=true
    ScaleRGB=true
   ) {module=ScaleColor001}
  ]
  rendererClasses [&NiagaraMeshRendererProperties]
  renderers [{rendererClass=&NiagaraMeshRendererProperties, rendererIndex=0}]
 }
 SpeedLine_Hands {
  emitterSpawnScript [&FN_PreserveSkydiveUserVariables() {module=PreserveUserVariables}]
  emitterUpdateScript [
   &EmitterState(Life Cycle Mode=::NewEnumerator1, Loop Duration=0.5, Scalability Mode=::NewEnumerator1, UseLoopDelay=false)
   &SpawnRate(
    SpawnRate=&Lerp_Float(
     A=&FN_FrontEndFloatModifier(Gameplay=0, Front End=20)
     B=&FN_FrontEndFloatModifier(Gameplay=16, Front End=60)
     Alpha=$System.Dive
    )
    Use Spawn Probability=false
   ) {enabled=false}
   &SpawnRate(
    SpawnRate=&FN_FrontEndFloatModifier(
     Gameplay=&Lerp_Float(A=0, B=16, Alpha=$System.Dive)
     Front End=&Lerp_Float(A=0, B=60, Alpha=$User.ParaGlideLeanAlpha)
    )
    Use Spawn Probability=false
   ) {module=SpawnRate002}
   &SpawnBurst_Instantaneous(
    Spawn Count=&MakeIntFromFloat(Float=&FN_FrontEndFloatModifier(Gameplay=&Lerp_Float(A=0, B=8, Alpha=$System.Dive), Front End=0))
    Use Spawn Probability=false
   )
   &SpawnRate(SpawnGroup=1, SpawnRate=8, Use Spawn Probability=false) {enabled=false, module=SpawnRate001}
  ]
  particleSpawnScript [
   &FN_SetSpawnGroup_v2(Number of Spawn Groups=4)
   /SetVariables(
    Particles.Dy_Darkness:NiagaraFloat=&UniformRangedFloat(Minimum=30, Maximum=1000)
    Particles.Dy_U_Tile:NiagaraFloat=&UniformRangedFloat(Minimum=0.5, Maximum=1)
    Particles.Dy_offsetUV:NiagaraFloat=&UniformRangedFloat(Minimum=0, Maximum=1)
    Particles.FallingSpeed:Vector3f=$Engine.Owner.Velocity
    Particles.PositionRandom:Vector3f=&UniformRangedVector(Minimum=<-2.5 -2.5 0>, Maximum=<2.5 2.5 0>)
   ) {isSetParameters=true}
   &InitializeParticle(
    Color=<rgba 1 0.1583 0.0084 1>
    Lifetime=&UniformRangedFloat(Minimum=0.4, Maximum=0.6)
    Mesh Scale=&Multiply_Vector(
     A=<1 1 1>
     B=&VectorFromFloat(Value=&UniformRangedFloat(Minimum=0.2, Maximum=1))
    )
    Module.Write Color=true
    Module.Write Lifetime=true
    Module.Write Mass=true
    Module.Write Material Random=true
    Module.Write Position=false
    Module.Write Scale=false
    Module.Write SpriteRotation=true
    Module.Write SpriteSize=true
    Position=&SimulationPosition()
    Sprite Rotation=90
    Sprite Size=&MakeVector2D(
     X=&Multiply_Float(A=15, B=&Lerp_Float(A=0.3, B=1, Alpha=$Particles.Dy_U_Tile))
     Y=&Multiply_Float(
      A=&UniformRangedFloat(Minimum=100, Maximum=150)
      B=&FloatFromCurve(
       FloatCurve=DataInterface() {propertyValues {Curve {keys [<at 0 1>,<at 1 2.5>]}}}
       CurveIndex=&Divide_Float(A=&VectorLength(Vector=$Particles.FallingSpeed), B=9000)
      )
     )
    )
   )
   &DynamicMaterialParameters(
    Index 0 Param 1=$Particles.Dy_offsetUV
    Index 0 Param 2=$Particles.Dy_U_Tile
    Index 0 Param 3=$Particles.Dy_Darkness
    Index 0 Param 4=1
    Module.Param0WriteEnabled=true
    Module.Param1WriteEnabled=true
    Module.Param2WriteEnabled=true
    Module.Param3WriteEnabled=true
   )
   &CameraOffset(Camera Offset Amount=&UniformRangedFloat(Minimum=-5, Maximum=5))
  ]
  particleUpdateScript [
   &SkeletalMeshLocation(
    Bone / Socket Index=&SpawnGroup()
    Bone Sampling Mode=::NewEnumerator4
    Inherit Velocity=false
    MaskSpawnGroup=false
    Skeletal Mesh=$User.SkeletalMesh
   )
   &ParticleState()
   &AddVelocity(
    Coordinate Space=::World
    Scale Added Velocity=&VectorFromFloat(Value=&Lerp_Float(A=-10, B=-50, Alpha=$System.Dive))
    Velocity=&NormalizeVector(Vector To Normalize=$Engine.Owner.Velocity)
   )
   &SolveForcesAndVelocity(Clamp Velocity=false, Limit Acceleration=false)
   &ScaleColor(
    Scale Alpha=&FloatFromCurve(
     FloatCurve=DataInterface() {propertyValues {Curve {keys [<at 0 0 RCIM_Cubic>,<at 0.3634 1.0074 RCIM_Cubic>,<at 1 0 RCIM_Cubic>]}}}
     Scale Curve=0.3
    )
    ScaleA=true
    ScaleRGB=true
   )
  ]
  rendererClasses [&NiagaraSpriteRendererProperties]
  renderers [{rendererClass=&NiagaraSpriteRendererProperties}]
 }
 Stars {
  emitterUpdateScript [
   &EmitterState(Life Cycle Mode=::NewEnumerator1, Loop Duration=1, UseLoopDelay=false)
   &SpawnRate(SpawnRate=&Lerp_Float(A=20, B=150, Alpha=$System.Dive), Use Spawn Probability=true)
  ]
  particleSpawnScript [
   &InitializeParticle(
    Color=<rgba 5 5 5 1>
    Lifetime=3
    Module.Write Color=true
    Module.Write Lifetime=true
    Module.Write Mass=true
    Module.Write Material Random=true
    Module.Write Position=false
    Module.Write Scale=true
    Module.Write SpriteRotation=false
    Module.Write SpriteSize=true
    Position=&SimulationPosition()
    Sprite Size=&RandomRangeVector2D(Minimum=<2 10>, Maximum=<4 18>)
   )
   /SetVariables(Particles.SpriteSizeDots:Vector2f=&Vector2DFromFloat(Value=&RandomRangeFloat(Minimum=0.2, Maximum=0.2))) {isSetParameters=true, module=SetVariables_#4}
   /SetVariables(Particles.ColorDots:LinearColor=&MakeLinearColorFromVectorAndFloat(
     "Vector (RGB)"=&Lerp_Vector(
      A=&VectorFromCurve(
       VectorCurve=DataInterface() {propertyValues {
         XCurve {keys [<at 0.0017 0.6254 RCIM_Constant>,<at 0.1205 0.0068 RCIM_Constant>,<at 0.3657 1.0315 RCIM_Constant>,<at 0.587 1 RCIM_Constant>,<at 0.7735 0.4033 RCIM_Constant>]}
         YCurve {keys [<at -0.0001 0.0068 RCIM_Constant>,<at 0.1211 0.1586 RCIM_Constant>,<at 0.3633 0.1422 RCIM_Constant>,<at 0.5891 1 RCIM_Constant>,<at 0.7812 0.1373 RCIM_Constant>,<at 0.9988 0.266 RCIM_Constant>]}
         ZCurve {keys [<at 0 1 RCIM_Constant>,<at 0.127 0.7998 RCIM_Constant>,<at 0.365 0.4415 RCIM_Constant>,<at 0.5907 1 RCIM_Constant>,<at 0.9986 0.9855 RCIM_Constant>]}
        }}
       CurveIndex=$Emitter.NormalizedLoopAge
      )
      B=&MakeVectorFromLinearColorRGB(Color=<rgba 0 3.6798 10 1>)
      Alpha=$Particles.DiveCLAMP
     )
     "Float (Alpha)"=&UniformRangedFloat(Minimum=0.2, Maximum=0.2)
    )) {isSetParameters=true, module=SetVariables_#5}
   /Niagara/Modules/Spawn/Location/V2/SphereLocation(
    MaskSpawnGroup=false
    Non Uniform Scale=<1 1 0>
    Offset=&Multiply_VectorByFloat(
     Vector=&NormalizeVector(Vector To Normalize=$Engine.Owner.Velocity)
     Float=&Lerp_Float(A=20, B=150, Alpha=$System.Dive)
    )
    Offset Coordinate Space:ENiagaraCoordinateSpace=::World
    Override Seed=false
    Sphere Radius=200
    Surface Only=false
    UseNonUniformScale=true
    UseOffset=true
    UseSphereOrientation=false
   )
   &InitialMeshOrientation(
    Orientation Vector=<0 0 1>
    Rotation=&RandomRangeVector(Minimum=<0 0 0>, Maximum=<1 0 0>)
    Use Orientation Vector=false
    Use Rotation Vector=true
   )
  ]
  particleUpdateScript [
   &AddVelocity(
    Coordinate Space=::World
    Scale Added Velocity=&VectorFromFloat(Value=&Lerp_Float(A=-10, B=-50, Alpha=$System.Dive))
    Velocity=&NormalizeVector(Vector To Normalize=$Engine.Owner.Velocity)
   )
   &AddVelocity(Velocity=&FN_FrontEndVector3Modifier(Gameplay=<0 0 0>, Front End=<0 0 20>)) {module=AddVelocity001}
   &ParticleState()
   &SolveForcesAndVelocity(Clamp Velocity=false, Limit Acceleration=false)
   &ScaleSpriteSize(Scale Factor=&MakeVector2D(
     X=&Lerp_Float(A=1, B=0.5, Alpha=$System.Dive)
     Y=&Lerp_Float(A=1, B=&FN_FrontEndFloatModifier(Gameplay=5, Front End=10), Alpha=$System.Dive)
    ))
   &ScaleSpriteSize(Scale Factor=&Vector2DFromFloat(Value=&FloatFromCurve(FloatCurve=DataInterface() {propertyValues {Curve {keys [
          {arriveTangent=6.9773, interpMode=RCIM_Cubic, leaveTangent=4.681, tangentMode=RCTM_User, time=0, value=0}
          {arriveTangentWeight=0.3334, interpMode=RCIM_Cubic, leaveTangent=1, leaveTangentWeight=0.3334, time=0.5, value=1}
         ]}}}))) {module=ScaleSpriteSize002}
   &ScaleMeshSize(Scale Factor=&Lerp_Vector(A=<0.2 1 1>, B=<10 1 1>, Alpha=$System.Dive))
   /SetVariables(Particles.SpriteSizeDots:Vector2f=&Vector2DFromFloat(Value=&Lerp_Float(A=&RandomRangeFloat(Minimum=5, Maximum=5), B=2, Alpha=$System.Dive))) {isSetParameters=true, module=SetVariables_#6}
   &ScaleColor(
    Scale RGB=&VectorFromCurve(
     VectorCurve=DataInterface() {propertyValues {
       XCurve {keys [<at 0.0017 0.6254 RCIM_Constant>,<at 0.1205 0.0068 RCIM_Constant>,<at 0.3657 1.0315 RCIM_Constant>,<at 0.587 1 RCIM_Constant>,<at 0.7735 0.4033 RCIM_Constant>]}
       YCurve {keys [<at -0.0001 0.0068 RCIM_Constant>,<at 0.1211 0.1586 RCIM_Constant>,<at 0.3633 0.1422 RCIM_Constant>,<at 0.5891 1 RCIM_Constant>,<at 0.7812 0.1373 RCIM_Constant>,<at 0.9988 0.266 RCIM_Constant>]}
       ZCurve {keys [<at 0 1 RCIM_Constant>,<at 0.127 0.7998 RCIM_Constant>,<at 0.365 0.4415 RCIM_Constant>,<at 0.5907 1 RCIM_Constant>,<at 0.9986 0.9855 RCIM_Constant>]}
      }}
     CurveIndex=&ReturnNormalizedExecIndex()
    )
    ScaleA=true
    ScaleRGB=true
   )
   &ScaleColor(
    Scale Alpha=150
    Scale RGB=&MakeVectorFromLinearColorRGB(Color=&ColorFromCurve(ColorCurve=DataInterface() {propertyValues {
        AlphaCurve {keys [<at 0 1 RCIM_Cubic>,<at 1 1 RCIM_Cubic>]}
        BlueCurve {keys [<at 0.0739 3>,<at 0.1439 1>,<at 0.2694 1 RCIM_Cubic>]}
        GreenCurve {keys [<at 0.0739 0.7932>,<at 0.1439 0.2644>,<at 0.2694 0.0904 RCIM_Cubic>]}
        RedCurve {keys [<at 0.0739 0.1702>,<at 0.1439 0.0567>,<at 0.2694 0.9924 RCIM_Cubic>]}
       }}))
    ScaleA=true
    ScaleRGB=true
   ) {module=ScaleColor001}
   &ScaleColor(
    Scale RGB=&VectorFromFloat(Value=&RandomRangeFloat(Minimum=1, Maximum=10))
    ScaleA=true
    ScaleRGB=true
   ) {module=ScaleColor002}
   &OrientMeshToVector(Look At Direction:Vector3f=$Particles.Velocity, Override Quat=false, UseOptionalSideAxis=false, UseOptionalUpAxis=false)
   /SetVariables(Particles.MeshOrientation=&MultiplyQuaternion(
     Quaternion A=&FindQuatBetweenTwoVectors(From Vector=<1 0 0>, To Vector=$Particles.Velocity)
     Quaternion B=$Particles.InitialRotation
    )) {enabled=false, isSetParameters=true, module=SetVariables_#7}
  ]
  rendererClasses [&NiagaraSpriteRendererProperties, &NiagaraMeshRendererProperties]
  renderers [
   {rendererClass=&NiagaraSpriteRendererProperties}
   {rendererClass=&NiagaraSpriteRendererProperties, rendererIndex=1}
   {rendererClass=&NiagaraMeshRendererProperties, rendererIndex=2}
   {rendererClass=&NiagaraSpriteRendererProperties, rendererIndex=3}
  ]
 }
 StarsMeshBadOrient001 {
  bEnabled=false
  emitterUpdateScript [
   &EmitterState(Life Cycle Mode=::NewEnumerator1, Loop Duration=1, UseLoopDelay=false)
   &SpawnRate(SpawnRate=&Lerp_Float(A=5, B=30, Alpha=$System.Dive), Use Spawn Probability=true)
  ]
  particleSpawnScript [
   &InitializeParticle(
    Color=<rgba 1.1012 6.1935 10 1>
    Lifetime=5
    Module.Write Color=true
    Module.Write Lifetime=true
    Module.Write Mass=true
    Module.Write Material Random=true
    Module.Write Position=false
    Module.Write Scale=true
    Module.Write SpriteRotation=false
    Module.Write SpriteSize=true
    Position=&SimulationPosition()
    Sprite Size=&RandomRangeVector2D(Minimum=<2 5>, Maximum=<4 20>)
   )
   /SetVariables(Particles.SpriteSizeDots:Vector2f=&Vector2DFromFloat(Value=&RandomRangeFloat(Minimum=1, Maximum=1))) {isSetParameters=true, module=SetVariables_#4}
   /SetVariables() {isSetParameters=true}
   /SetVariables(Particles.ColorDots:LinearColor=&MakeLinearColorFromVectorAndFloat(
     "Vector (RGB)"=&Lerp_Vector(
      A=&VectorFromCurve(
       VectorCurve=DataInterface() {propertyValues {
         XCurve {keys [<at 0.0017 0.6254 RCIM_Constant>,<at 0.1205 0.0068 RCIM_Constant>,<at 0.3657 1.0315 RCIM_Constant>,<at 0.587 1 RCIM_Constant>,<at 0.7735 0.4033 RCIM_Constant>]}
         YCurve {keys [<at -0.0001 0.0068 RCIM_Constant>,<at 0.1211 0.1586 RCIM_Constant>,<at 0.3633 0.1422 RCIM_Constant>,<at 0.5891 1 RCIM_Constant>,<at 0.7812 0.1373 RCIM_Constant>,<at 0.9988 0.266 RCIM_Constant>]}
         ZCurve {keys [<at 0 1 RCIM_Constant>,<at 0.127 0.7998 RCIM_Constant>,<at 0.365 0.4415 RCIM_Constant>,<at 0.5907 1 RCIM_Constant>,<at 0.9986 0.9855 RCIM_Constant>]}
        }}
       CurveIndex=$Emitter.NormalizedLoopAge
      )
      B=&MakeVectorFromLinearColorRGB(Color=<rgba 0 3.6798 10 1>)
      Alpha=$Particles.DiveCLAMP
     )
     "Float (Alpha)"=&UniformRangedFloat(Minimum=0.2, Maximum=0.2)
    )) {isSetParameters=true, module=SetVariables_#5}
   &SkeletalMeshLocation(Inherit Velocity=false, MaskSpawnGroup=false, Module.Override Randomness=false, Skeletal Mesh=$User.SkeletalMesh)
   &SphereLocation(
    MaskSpawnGroup=false
    Non Uniform Scale=<1 1 0>
    Offset=<0 0 -100>
    Override Seed=false
    Surface Only=false
    UseNonUniformScale=true
    UseOffset=true
    UseSphereOrientation=false
   )
   &KillParticlesInVolume(Invert Volume=true, Sphere Radius=50) {enabled=false}
   &InitialMeshOrientation(
    Orientation Vector=<0 1 0>
    Rotation=&RandomRangeVector(Minimum=<0 0 0>, Maximum=<1 0 0>)
    Use Orientation Vector=false
    Use Rotation Vector=false
   )
  ]
  particleUpdateScript [
   &AddVelocity(
    Coordinate Space=::World
    Scale Added Velocity=&VectorFromFloat(Value=&Lerp_Float(A=-10, B=-50, Alpha=$System.Dive))
    Velocity=&NormalizeVector(Vector To Normalize=$Engine.Owner.Velocity)
   )
   &ParticleState()
   &SolveForcesAndVelocity(Clamp Velocity=false, Limit Acceleration=false)
   /NiagaraFN/Sprite/FN_SpriteAlignmentAndFacing(SpriteAlignment:Vector3f=<0 0 1>) {enabled=false, module=SpriteAlignmentAndFacing}
   &ScaleSpriteSize(Scale Factor=&MakeVector2D(
     X=&Lerp_Float(A=1, B=0.2, Alpha=$System.Dive)
     Y=&Lerp_Float(A=1, B=5, Alpha=$System.Dive)
    ))
   &ScaleMeshSize(Scale Factor=&Lerp_Vector(A=<1 1 1>, B=<12 1 1>, Alpha=$System.Dive))
   /SetVariables(Particles.SpriteSizeDots:Vector2f=&Vector2DFromFloat(Value=&Lerp_Float(A=&RandomRangeFloat(Minimum=10, Maximum=30), B=0.2, Alpha=$System.Dive))) {isSetParameters=true, module=SetVariables_#6}
   &ScaleColor(
    Scale RGB=&VectorFromCurve(
     VectorCurve=DataInterface() {propertyValues {
       XCurve {keys [<at 0.0017 0.6254 RCIM_Constant>,<at 0.1205 0.0068 RCIM_Constant>,<at 0.3657 1.0315 RCIM_Constant>,<at 0.587 1 RCIM_Constant>,<at 0.7735 0.4033 RCIM_Constant>]}
       YCurve {keys [<at -0.0001 0.0068 RCIM_Constant>,<at 0.1211 0.1586 RCIM_Constant>,<at 0.3633 0.1422 RCIM_Constant>,<at 0.5891 1 RCIM_Constant>,<at 0.7812 0.1373 RCIM_Constant>,<at 0.9988 0.266 RCIM_Constant>]}
       ZCurve {keys [<at 0 1 RCIM_Constant>,<at 0.127 0.7998 RCIM_Constant>,<at 0.365 0.4415 RCIM_Constant>,<at 0.5907 1 RCIM_Constant>,<at 0.9986 0.9855 RCIM_Constant>]}
      }}
     CurveIndex=$Emitter.NormalizedLoopAge
    )
    ScaleA=true
    ScaleRGB=true
   )
   &ScaleSpriteSize(Scale Factor=<3 3>) {enabled=false, module=ScaleSpriteSize001}
   &OrientMeshToVector(Look At Direction:Vector3f=$Particles.Velocity, Override Quat=false, UseOptionalSideAxis=false, UseOptionalUpAxis=false)
   /SetVariables(Particles.MeshOrientation=&MultiplyQuaternion(
     Quaternion A=&FindQuatBetweenTwoVectors(From Vector=<1 0 0>, To Vector=$Particles.Velocity)
     Quaternion B=$Particles.InitialRotation
    )) {enabled=false, isSetParameters=true, module=SetVariables_#7}
  ]
  rendererClasses [&NiagaraSpriteRendererProperties, &NiagaraMeshRendererProperties]
  renderers [
   {rendererClass=&NiagaraSpriteRendererProperties}
   {rendererClass=&NiagaraSpriteRendererProperties, rendererIndex=1}
   {rendererClass=&NiagaraMeshRendererProperties, rendererIndex=2}
  ]
 }
 crystal_shards_SpeedDebris {
  bEnabled=false
  emitterUpdateScript [
   &EmitterState(Life Cycle Mode=::NewEnumerator1, Loop Duration=1, UseLoopDelay=false)
   &SpawnRate(Spawn Probability=0.6, SpawnRate=30, Use Spawn Probability=true)
  ]
  particleSpawnScript [
   &FN_DivingAttributes_v2(Accumulate Dive=0.1, Accumulate Rest=-0.1, SkeletalMesh:NiagaraDataInterfaceSkeletalMesh=$User.SkeletalMesh)
   /SetVariables(Particles.DiveCLAMP:NiagaraFloat=&ClampFloat(Float=&FN_FrontEndFloatModifier(Gameplay=$Particles.Dive, Front End=&Multiply_Float(A=$Particles.Dive, B=10)))) {isSetParameters=true}
   &InitializeParticle(
    Color=&LerpLinearColors(StartColor=<rgba 28.3334 0 50 1>, EndColor=<rgba 0 40.4868 50 1>, LerpFactor=$Particles.DiveCLAMP)
    Lifetime=&UniformRangedFloat(Minimum=0.3, Maximum=0.9)
    Module.Write Color=true
    Module.Write Lifetime=true
    Module.Write Mass=false
    Module.Write Material Random=true
    Module.Write Position=false
    Module.Write Scale=false
    Module.Write SpriteRotation=false
    Module.Write SpriteSize=true
    Sprite Size=&Vector2DFromFloat(Value=&Multiply_Float(
      A=20
      B=&Lerp_Float(
       A=&UniformRangedFloat(Minimum=0.5, Maximum=1)
       B=&UniformRangedFloat(Minimum=1, Maximum=1.25)
       Alpha=$Particles.DiveCLAMP
      )
     ))
   )
   &AddVelocity(Velocity=&MakeVector(
     X=&Lerp_Float(
      A=&UniformRangedFloat(Minimum=-20, Maximum=20)
      B=&UniformRangedFloat(Minimum=-20, Maximum=20)
      Alpha=$Particles.DiveCLAMP
     )
     Y=0
     Z=0
    ))
   &SkeletalMeshLocation(
    Bone Sampling Mode=::NewEnumerator2
    Inherit Velocity=false
    MaskSpawnGroup=false
    Module.Override Randomness=false
    Skeletal Mesh=DataInterface() {propertyValues {
      DefaultMesh=null
      ExcludeBoneName=root
      OptionalComponentClass=null
      PreviewMesh=&M_Med_Soldier_04_Preview
      SkinningMode=SkinOnTheFly
      SoftSourceActor=null
      SourceMode=Default
      UvSetIndex=0
      WholeMeshLOD=-1
      bExcludeBone=false
      bReadDeformedGeometry=true
      bRequireCurrentFrameData=true
      ComponentTags []
      FilteredBones [root, ik_foot_root, ik_hand_root, VB Root_hand_l, VB Root_hand_r, attach, VB root_prop]
      FilteredSockets [None]
      MeshUserParameter {parameter {name=None, typeDefHandle {registeredTypeIndex=1}, varData []}}
      SamplingRegions []
     }}
   )
   &CylinderLocation(
    Cylinder Height=30
    Cylinder Radius=&UniformRangedFloat(Minimum=5, Maximum=30)
    MaskSpawnGroup=false
    Module.Enable Lathe Profile=false
    Module.Override Randomness=false
    Module.Surface Only=true
    Orientation Axis:ENiagaraOrientationAxis=::XAxis
    Use Endcaps In Surface Only Mode=false
   )
  ]
  particleUpdateScript [
   /SetVariables(Particles.DiveCLAMP:NiagaraFloat=&ClampFloat(Float=&FN_FrontEndFloatModifier(Gameplay=$Particles.Dive, Front End=&Multiply_Float(A=$Particles.Dive, B=10)))) {isSetParameters=true}
   &ParticleState()
   &AddVelocity(Velocity=&MakeVector(
     X=0
     Y=0
     Z=&Lerp_Float(
      A=&UniformRangedFloat(Minimum=30, Maximum=45)
      B=&UniformRangedFloat(Minimum=60, Maximum=90)
      Alpha=Linked() {linkedVariable {name=Particles.IntegratedDiveCLAMP+, type=NiagaraFloat}}
     )
    )) {module=AddVelocity002}
   &SolveForcesAndVelocity(Clamp Velocity=false, Limit Acceleration=false)
   &ScaleColor(
    Scale Alpha=&FloatFromCurve(FloatCurve=DataInterface() {propertyValues {Curve {keys [<at 0 1>,<at 1 0>]}}})
    Scale RGB=&VectorFromFloat(Value=&FloatFromCurve(FloatCurve=DataInterface() {propertyValues {Curve {keys [<at 0 1 RCIM_Cubic>, {arriveTangent=-1.8097, interpMode=RCIM_Cubic, leaveTangent=-1.8097, tangentMode=RCTM_User, time=1, value=0}]}}}))
    ScaleA=true
    ScaleRGB=true
   )
   &ScaleSpriteSize(Scale Factor=&Vector2DFromFloat(Value=&Multiply_Float(A=$Particles.NormalizedAge, B=&OneMinusFloat(Float=$Particles.NormalizedAge))))
   &ScaleSpriteSizeBySpeed(
    Max Scale Factor=<0.5 40>
    Min Scale Factor=<0.1 1>
    Velocity Threshold:NiagaraFloat=&Lerp_Float(A=1800, B=1200, Alpha=Linked() {linkedVariable {name=Particles.IntegratedDiveCLAMP+, type=NiagaraFloat}})
   )
  ]
  rendererClasses [&NiagaraSpriteRendererProperties]
  renderers [{rendererClass=&NiagaraSpriteRendererProperties}]
 }
 crystal_shards_blue {
  bEnabled=false
  emitterUpdateScript [
   &EmitterState(Life Cycle Mode=::NewEnumerator1, Loop Duration=1, UseLoopDelay=false)
   &SpawnRate(SpawnRate=60, Use Spawn Probability=false)
  ]
  particleSpawnScript [
   &FN_DivingAttributes_v2(Accumulate Dive=0.1, Accumulate Rest=-0.1, SkeletalMesh:NiagaraDataInterfaceSkeletalMesh=$User.SkeletalMesh)
   /SetVariables(Particles.DiveCLAMP:NiagaraFloat=&ClampFloat(Float=&FN_FrontEndFloatModifier(Gameplay=$Particles.Dive, Front End=&Multiply_Float(A=$Particles.Dive, B=10)))) {isSetParameters=true}
   &InitializeParticle(
    Color=&LerpLinearColors(StartColor=<rgba 28.3334 0 50 1>, EndColor=<rgba 0 40.4868 50 1>, LerpFactor=$Particles.DiveCLAMP)
    Lifetime=&UniformRangedFloat(Minimum=0.7, Maximum=1.1)
    Mesh Scale=&Multiply_Vector(
     A=<0.1 0.1 0.025>
     B=&VectorFromFloat(Value=&Lerp_Float(
       A=&UniformRangedFloat(Minimum=0.3, Maximum=0.7)
       B=&UniformRangedFloat(Minimum=0.1, Maximum=0.3)
       Alpha=$Particles.DiveCLAMP
      ))
    )
    Module.Write Color=true
    Module.Write Lifetime=true
    Module.Write Mass=false
    Module.Write Material Random=true
    Module.Write Position=true
    Module.Write Scale=true
    Module.Write SpriteRotation=false
    Module.Write SpriteSize=false
    Position=&Add_Vector(A=&SimulationPosition(), B=<0 0 0>)
    Sprite Size=<5 5>
   )
   &AddVelocity(Velocity=&MakeVector(
     X=0
     Y=0
     Z=&Lerp_Float(
      A=&UniformRangedFloat(Minimum=0, Maximum=0)
      B=&UniformRangedFloat(Minimum=600, Maximum=1200)
      Alpha=$Particles.DiveCLAMP
     )
    )) {module=AddVelocity001}
   &CylinderLocation(
    Cylinder Height=30
    Cylinder Radius=&UniformRangedFloat(Minimum=5, Maximum=30)
    MaskSpawnGroup=false
    Module.Enable Lathe Profile=false
    Module.Override Randomness=false
    Module.Surface Only=true
    Non Uniform Scale:Vector3f=<1.2 1.2 0.5>
    Orientation Axis:ENiagaraOrientationAxis=::XAxis
    Use Endcaps In Surface Only Mode=false
   )
   &InitialMeshOrientation(
    Orientation Vector=<0 0 1>
    Rotation=&UniformRangedVector(Minimum=<0 0 0>, Maximum=<1 0 0>)
    Use Orientation Vector=true
    Use Rotation Vector=true
   )
   &SkeletalMeshLocation(
    Bone Sampling Mode=::NewEnumerator2
    Inherit Velocity=false
    MaskSpawnGroup=false
    Module.Override Randomness=false
    Skeletal Mesh=DataInterface() {propertyValues {
      DefaultMesh=null
      ExcludeBoneName=root
      OptionalComponentClass=null
      PreviewMesh=&M_Med_Soldier_04_Preview
      SkinningMode=SkinOnTheFly
      SoftSourceActor=null
      SourceMode=Default
      UvSetIndex=0
      WholeMeshLOD=-1
      bExcludeBone=false
      bReadDeformedGeometry=true
      bRequireCurrentFrameData=true
      ComponentTags []
      FilteredBones [root, ik_foot_root, ik_hand_root, VB Root_hand_l, VB Root_hand_r, attach, VB root_prop]
      FilteredSockets [None]
      MeshUserParameter {parameter {name=None, typeDefHandle {registeredTypeIndex=1}, varData []}}
      SamplingRegions []
     }}
   )
  ]
  particleUpdateScript [
   /SetVariables(Particles.DiveCLAMP:NiagaraFloat=&ClampFloat(Float=&FN_FrontEndFloatModifier(Gameplay=$Particles.Dive, Front End=&Multiply_Float(A=$Particles.Dive, B=10)))) {isSetParameters=true}
   &ParticleState()
   &AddVelocity(Velocity=&MakeVector(
     X=0
     Y=0
     Z=&Lerp_Float(
      A=&UniformRangedFloat(Minimum=3, Maximum=15)
      B=&UniformRangedFloat(Minimum=30, Maximum=45)
      Alpha=$Particles.DiveCLAMP
     )
    ))
   /Niagara/Modules/Update/Forces/CurlNoiseForce(Noise Frequency=5, Noise Strength=150, Pan Noise=false, Random Seed=1)
   &SolveForcesAndVelocity(Clamp Velocity=false, Limit Acceleration=false)
   &ScaleMeshSize(Scale Factor=&VectorFromFloat(Value=&FloatFromCurve(FloatCurve=DataInterface() {propertyValues {Curve {keys [
          <at 0 1>
          {arriveTangent=-0.5431, interpMode=RCIM_Cubic, leaveTangent=-0.5431, tangentMode=RCTM_User, time=0.6, value=1}
          {arriveTangent=-5.5854, interpMode=RCIM_Cubic, leaveTangent=-5.5854, tangentMode=RCTM_User, time=1, value=0}
         ]}}})))
   &ScaleColor(
    Scale Alpha=&FloatFromCurve(FloatCurve=DataInterface() {propertyValues {Curve {keys [<at 0 1>,<at 1 0>]}}})
    Scale RGB=&VectorFromFloat(Value=&FloatFromCurve(FloatCurve=DataInterface() {propertyValues {Curve {keys [<at 0 1>,<at 1 0>]}}}))
    ScaleA=true
    ScaleRGB=true
   )
   &ScaleMeshSizeBySpeed(
    Max Scale Factor=<100 0.6 0.6>
    Sample Scale Factor By Curve=true
    Scale Factor Curve=DataInterface() {propertyValues {Curve {keys [<at 0 0>,<at 0.3 0>,<at 1 1>]}}}
    Speed Threshold=900
   )
  ]
  rendererClasses [&NiagaraMeshRendererProperties]
  renderers [{rendererClass=&NiagaraMeshRendererProperties, rendererIndex=0}]
 }
}
summary {
 systemName=NS_Contrail_Starfield
 emitters [
  {emitterName=Body_Decal, rendererClasses [&NiagaraSpriteRendererProperties]}
  {bEnabled=false, emitterName=crystal_shards_blue, rendererClasses [&NiagaraMeshRendererProperties]}
  {bEnabled=false, emitterName=SpeedCone_Main, rendererClasses [&NiagaraMeshRendererProperties]}
  {bEnabled=false, emitterName=crystal_shards_SpeedDebris, rendererClasses [&NiagaraSpriteRendererProperties]}
  {bEnabled=false, emitterName=Fire, rendererClasses [&NiagaraSpriteRendererProperties]}
  {bEnabled=false, emitterName=Blue_Main, rendererClasses [&NiagaraMeshRendererProperties]}
  {bEnabled=false, emitterName=Blue_Main_Body, rendererClasses [&NiagaraMeshRendererProperties]}
  {bEnabled=false, emitterName=Fire_Upward, rendererClasses [&NiagaraSpriteRendererProperties]}
  {emitterName=Stars, rendererClasses [&NiagaraSpriteRendererProperties, &NiagaraMeshRendererProperties]}
  {emitterName=Mesh_Dish, rendererClasses [&NiagaraMeshRendererProperties]}
  {
   bEnabled=false
   emitterName=StarsMeshBadOrient001
   rendererClasses [&NiagaraSpriteRendererProperties, &NiagaraMeshRendererProperties]
  }
  {emitterName=SpeedLine_Hands, rendererClasses [&NiagaraSpriteRendererProperties]}
  {
   emitterName=Electric_Hands_and_Feet
   rendererClasses [&NiagaraSpriteRendererProperties, &NiagaraMeshRendererProperties]
  }
 ]
 userVariables [
  User.bIsFrontEnd {type=NiagaraBool}
  User.bIsFrontEndPreview {type=NiagaraBool}
  User.ColorChange {type=LinearColor, description="", defaultValue {struct=LinearColor, value=<rgba 1 0.0469 0.6556 1>}}
  User.ParaGlideLeanAlpha {type=NiagaraFloat, defaultValue=0, description=""}
  User.PlayerSpeed {type=NiagaraFloat, defaultValue=0, description=""}
  User.RibbonIDOffset {type=NiagaraInt32}
  User.SizeScale {type=NiagaraFloat, defaultValue=0, description=""}
  User.SkeletalMesh {
   type=NiagaraDataInterfaceSkeletalMesh
   defaultValue {
    struct=NiagaraExt_VariableValue_DataInterface
    value {dataInterface=&self:NiagaraDataInterfaceSkeletalMesh_2}
   }
  }
 ]
}
